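# Physics asset for the face accessory part of the large female LeatherJacketPurple_Music body, covering its dynamic hair bones.
# Small capsules on two mirrored two-bone chains, left and right, hang from a larger head capsule.
# Each root bone is constrained to the head and each tip to its root, with limited swing and twist.
# The right root body carries a second, identical capsule named after the left one.
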PhysicsAsset /BRCosmetics/Characters/Player/Female/Large/Bodies/F_LRG_LeatherJacketPurple_Music/Meshes/Parts/F_LRG_LeatherJacketPurple_Music_FaceAcc_Physics
bodies {
 dyn_hair_a_01_l [{shapeType=Capsule, length=1, radius=1, shapeName=dyn_hair_a_01_l_capsule, center=<1 0 0>, rotation=<rot -90 0 0>}]
 dyn_hair_a_01_r [
  {shapeType=Capsule, length=1, radius=1, shapeName=dyn_hair_a_01_r_capsule, center=<1 0 0>, rotation=<rot -90 0 0>}
  {shapeType=Capsule, length=1, radius=1, shapeName=dyn_hair_a_01_l_capsule, center=<1 0 0>, rotation=<rot -90 0 0>}
 ]
 dyn_hair_a_02_l [{shapeType=Capsule, length=1, radius=1, shapeName=dyn_hair_a_02_l_capsule, center=<1 0 0>, rotation=<rot -90 0 0>}]
 dyn_hair_a_02_r [{shapeType=Capsule, length=1, radius=1, shapeName=dyn_hair_a_02_r_capsule, center=<1 0 0>, rotation=<rot -90 0 0>}]
 head [{shapeType=Capsule, length=2, radius=7, shapeName=head_capsule, center=<1 0 0>, rotation=<rot -90 0 0>}]
}
constraints [
 {bone1Name=dyn_hair_a_01_r, bone2Name=head, swing1LimitDegrees=20, swing2LimitDegrees=15, twistLimitDegrees=10}
 {bone1Name=dyn_hair_a_01_l, bone2Name=head, swing1LimitDegrees=20, swing2LimitDegrees=15, twistLimitDegrees=10}
 {bone1Name=dyn_hair_a_02_l, bone2Name=dyn_hair_a_01_l, swing1LimitDegrees=20, swing2LimitDegrees=25, twistLimitDegrees=10}
 {bone1Name=dyn_hair_a_02_r, bone2Name=dyn_hair_a_01_r, swing1LimitDegrees=20, swing2LimitDegrees=25, twistLimitDegrees=10}
]
