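# Physics asset for the hood on the face accessory mesh of the large female player's purple leather jacket "Music" cosmetic body.
# It has two capsule bodies: a tiny one on the dyn_hood bone and a much larger one on spine_05 for the upper torso.
# A single constraint joins the hood to the spine with tight, equal swing and twist angle limits, so the hood can rotate only slightly relative to the torso.
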PhysicsAsset /BRCosmetics/Characters/Player/Female/Large/Bodies/F_LRG_LeatherJacketPurple_Music/Meshes/Parts/F_LRG_LeatherJacketPurple_Music_FaceAcc_Physics_hood
bodies {
 dyn_hood [{shapeType=Capsule, length=1, radius=1, shapeName=dyn_hood_capsule, center=<1 0 0>, rotation=<rot -90 0 0>}]
 spine_05 [{shapeType=Capsule, length=15, radius=10, shapeName=spine_05_capsule, center=<5 0 0>}]
}
constraints [{bone1Name=dyn_hood, bone2Name=spine_05, swing1LimitDegrees=8, swing2LimitDegrees=8, twistLimitDegrees=8}]
